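# Physics asset for the MollyKit offhand pickaxe mesh, defining a four-bone tassel that hangs from the weapon handle.
# A larger offset sphere on Melee_TwoHanded_Handle anchors the setup, and each dyn_Tassel bone carries a small sphere body.
# Angular constraints link the tassel bodies in a chain, with wide swing at the handle joint, tighter swing down the chain and narrow twist throughout.
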PhysicsAsset /BRCosmetics/Weapons/FORT_Melee/Pickaxe_MollyKit_Offhand/Meshes/MollyKit_Offhand_Axe_Physics
bodies {
 Melee_TwoHanded_Handle [{shapeType=Sphere, radius=5, center=<-2.7199 0 -7.2658>}]
 dyn_Tassel_1 [{shapeType=Sphere, radius=1, shapeName=dyn_Tassel_1_sphere}]
 dyn_Tassel_2 [{shapeType=Sphere, radius=1, shapeName=dyn_Tassel_2_sphere}]
 dyn_Tassel_3 [{shapeType=Sphere, radius=1, shapeName=dyn_Tassel_3_sphere}]
 dyn_Tassel_4 [{shapeType=Sphere, radius=1, shapeName=dyn_Tassel_4_sphere}]
}
constraints [
 {bone1Name=dyn_Tassel_1, bone2Name=Melee_TwoHanded_Handle, swing1LimitDegrees=80, swing2LimitDegrees=80, twistLimitDegrees=10}
 {bone1Name=dyn_Tassel_2, bone2Name=dyn_Tassel_1, swing1LimitDegrees=40, swing2LimitDegrees=40, twistLimitDegrees=10}
 {bone1Name=dyn_Tassel_3, bone2Name=dyn_Tassel_2, swing1LimitDegrees=40, swing2LimitDegrees=40, twistLimitDegrees=10}
 {bone1Name=dyn_Tassel_4, bone2Name=dyn_Tassel_3, swing1LimitDegrees=40, swing2LimitDegrees=40, twistLimitDegrees=10}
]
