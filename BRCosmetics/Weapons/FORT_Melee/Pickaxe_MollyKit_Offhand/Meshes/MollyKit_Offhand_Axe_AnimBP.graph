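# Post-process-style Anim Blueprint for the MollyKit offhand pickaxe, adding secondary motion to its four-bone tassel on top of the input pose.
# Nested bool blends pick between a RigidBody simulation and two differently tuned AnimDynamics chains, cached as the output.
# EventGraph updates gravity override, movement-based velocity scale and front-end solver flags, though the AnimGraph shows no reads of them.
# An older AnimDynamics chain remains disconnected from the output.
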AnimBlueprint /BRCosmetics/Weapons/FORT_Melee/Pickaxe_MollyKit_Offhand/Meshes/MollyKit_Offhand_Axe_AnimBP
target_skeleton=&/MollyKit_Offhand_Axe_Skeleton
graphs [
 AnimGraph {nodes [
   n1 = Anim Dynamics - Bone: dyn_Tassel_1(Alpha=1, ExternalForce="0,0,0") {
    angularDampingOverride=0.9994
    bOverrideAngularDamping=true
    bUseGravityOverride=true
    alphaBoolBlend {blendOption=Linear}
    boundBone {boneName=dyn_Tassel_1}
    componentAppliedLinearAccClamp=<2000 2000 2000>
    componentLinearAccScale=<0.4 0.4 0.5>
    componentLinearVelScale=<0.4 0.4 0.5>
    gravityOverride {z=-980}
    physicsBodyDefinitions [{
      boundBone {boneName=dyn_Tassel_1}
      boxExtents=<8 8 8>
      constraintSetup {angularLimitsMax=<5 50 80>, angularLimitsMin=<-5 -20 -90>, angularTarget=<0 0 0>}
      localJointOffset=<1 0 0>
     }]
   }
   n2 = Anim Dynamics - Bone: dyn_Tassel_2(ComponentPose=@n1, Alpha=1, ExternalForce="0,0,0") {
    angularDampingOverride=0.9994
    bOverrideAngularDamping=true
    bUseGravityOverride=true
    alphaBoolBlend {blendOption=Linear}
    boundBone {boneName=dyn_Tassel_2}
    componentAppliedLinearAccClamp=<2000 2000 2000>
    componentLinearAccScale=<0.4 0.4 0.5>
    componentLinearVelScale=<0.4 0.4 0.5>
    gravityOverride {z=-980}
    physicsBodyDefinitions [{
      boundBone {boneName=dyn_Tassel_2}
      boxExtents=<8 8 8>
      constraintSetup {angularLimitsMax=<5 20 10>, angularLimitsMin=<-5 -20 -10>, angularTarget=<0 0 0>}
      localJointOffset=<1 0 0>
     }]
    planarLimits [{drivingBone {boneName=Melee_TwoHanded_Handle}, planeTransform {location=<0 0 -6>, rotation=<rot 0 0 -180>}}]
   }
   n3 = Anim Dynamics - Bone: dyn_Tassel_3(ComponentPose=@n2, Alpha=1, ExternalForce="0,0,0") {
    angularDampingOverride=0.9994
    bOverrideAngularDamping=true
    bUseGravityOverride=true
    bUseSphericalLimits=true
    alphaBoolBlend {blendOption=Linear}
    boundBone {boneName=dyn_Tassel_3}
    componentAppliedLinearAccClamp=<2000 2000 2000>
    componentLinearAccScale=<0.4 0.4 0.5>
    componentLinearVelScale=<0.4 0.4 0.5>
    gravityOverride {z=-980}
    physicsBodyDefinitions [{
      boundBone {boneName=dyn_Tassel_3}
      boxExtents=<8 8 8>
      constraintSetup {angularLimitsMax=<5 30 30>, angularLimitsMin=<-5 -30 -30>, angularTarget=<0 0 0>}
      localJointOffset=<1 0 0>
     }]
    planarLimits [{drivingBone {boneName=Melee_TwoHanded_Handle}, planeTransform {location=<0 0 -7>, rotation=<rot 0 0 180>}}]
   }
   n4 = Anim Dynamics - Bone: dyn_Tassel_4(ComponentPose=@n3, Alpha=1, ExternalForce="0,0,0") {
    angularDampingOverride=0.9994
    bOverrideAngularDamping=true
    bUseGravityOverride=true
    bUseSphericalLimits=true
    alphaBoolBlend {blendOption=Linear}
    boundBone {boneName=dyn_Tassel_4}
    componentAppliedLinearAccClamp=<2000 2000 2000>
    componentLinearAccScale=<0.4 0.4 0.5>
    componentLinearVelScale=<0.4 0.4 0.5>
    gravityOverride {z=-980}
    physicsBodyDefinitions [{
      boundBone {boneName=dyn_Tassel_4}
      boxExtents=<8 8 8>
      constraintSetup {angularLimitsMax=<5 30 30>, angularLimitsMin=<-5 -30 -30>, angularTarget=<0 0 0>}
      localJointOffset=<1 0 0>
     }]
    planarLimits [{drivingBone {boneName=Melee_TwoHanded_Handle}, planeTransform {location=<0 0 -7>, rotation=<rot 0 0 180>}}]
   }
   Component To Local(ComponentPose=@n4)
   "Transform (Modify) Bone - Bone: dyn_Tassel_3"(Alpha=1, Translation="0,0,0", Rotation="0,0,0", Scale="1,1,1") {rotationMode=BMM_Additive, translationMode=BMM_Additive, alphaBoolBlend {blendOption=Linear}, boneToModify {boneName=dyn_Tassel_3}}
   Output Pose(Result=Use cached pose 'Dynamics_Cache'()) {layerGroup=None}
   AnimGraphNode_SubInput_0 = Input Pose()
   Save cached pose 'input pose'(Pose=@AnimGraphNode_SubInput_0) {cacheName=input pose}
   n5 = RigidBody(ComponentPose=Local To Component(LocalPose=Use cached pose 'input pose'())) {
    overridePhysicsAsset=&/MollyKit_Offhand_Axe_Physics
    alphaBoolBlend {blendOption=Linear}
    componentAppliedLinearAccClamp=<3000 3000 3000>
    componentLinearAccScale=<1 1 1>
    componentLinearVelScale=<1 1 1>
    overrideWorldGravity {z=-980}
    simSpaceSettings {dampingAlpha=0, maxAngularAcceleration=1600, maxAngularVelocity=400, maxLinearAcceleration=1000, maxLinearVelocity=400, worldAlpha=1}
   }
   n6 = "Transform (Modify) Bone - Bone: dyn_Tassel_4"!(ComponentPose=@n5, Alpha=1, Rotation="-30,-50,0") -> (Pose) {rotationMode=BMM_Additive, rotationSpace=BCS_BoneSpace, boneToModify {boneName=dyn_Tassel_4}, rotation {pitch=-30, yaw=-50}}
   n7 = Component To Local(ComponentPose=@n6)
   n8 = "Transform (Modify) Bone - Bone: deform_Tassel_MSTR"(
    ComponentPose=Local To Component(LocalPose=Use cached pose 'input pose'())
    Alpha=1
    Translation="0,0,0"
    Rotation="-170,90,0"
    Scale="1,1,1"
   ) {rotationMode=BMM_Additive, rotationSpace=BCS_BoneSpace, boneToModify {boneName=deform_Tassel_MSTR}, rotation {pitch=-170, yaw=90}}
   n9 = Anim Dynamics - Bone: dyn_Tassel_1!(ComponentPose=@n8, Alpha=0.5, GravityOverride:Vector="0,0,-980") -> (Pose) {
    angularDampingOverride=1
    angularSpringConstant=500
    bOverrideAngularDamping=true
    bUseGravityOverride=true
    bUsePlanarLimit=false
    simulationSpace=BoneRelative
    alphaBoolBlend {blendOption=Linear}
    boundBone {boneName=dyn_Tassel_1}
    componentAppliedLinearAccClamp=<3000 3000 3000>
    componentLinearAccScale=<0.3 0.3 0.4>
    componentLinearVelScale=<0.3 0.3 0.4>
    gravityOverride {z=-980}
    physicsBodyDefinitions [{
      boundBone {boneName=dyn_Tassel_1}
      boxExtents=<6 12 12>
      constraintSetup {angularLimitsMax=<0 60 5>, angularLimitsMin=<0 -60 0>, angularTarget=<1 0 0>}
      localJointOffset=<2 0 0>
     }]
    relativeSpaceBone {boneName=b_AR_Root}
    simSpaceSettings {maxAngularAcceleration=25, maxAngularVelocity=5, simSpaceAngularAlpha=1}
   }
   n10 = Anim Dynamics - Bone: dyn_Tassel_2!(ComponentPose=@n9, Alpha=0.5, GravityOverride:Vector="0,0,-1500") -> (Pose) {
    angularDampingOverride=1
    angularSpringConstant=500
    bOverrideAngularDamping=true
    bUseGravityOverride=true
    bUsePlanarLimit=false
    simulationSpace=BoneRelative
    alphaBoolBlend {blendOption=Linear}
    boundBone {boneName=dyn_Tassel_2}
    componentAppliedLinearAccClamp=<3000 3000 3000>
    componentLinearAccScale=<0.2 0.3 0.3>
    componentLinearVelScale=<0.2 0.3 0.3>
    gravityOverride {z=-980}
    physicsBodyDefinitions [{
      boundBone {boneName=dyn_Tassel_2}
      boxExtents=<10 20 20>
      constraintSetup {angularLimitsMax=<20 100 60>, angularLimitsMin=<-20 -100 -60>, angularTarget=<1 0 0>}
      localJointOffset=<2.5 0 0>
     }]
    relativeSpaceBone {boneName=b_AR_Root}
    simSpaceSettings {maxAngularAcceleration=25, maxAngularVelocity=5, simSpaceAngularAlpha=1}
   }
   n11 = Anim Dynamics - Bone: dyn_Tassel_3!(ComponentPose=@n10, Alpha=0.5, GravityOverride:Vector="0,0,-1500") -> (Pose) {
    angularDampingOverride=1
    angularSpringConstant=500
    bOverrideAngularDamping=true
    bUseGravityOverride=true
    bUsePlanarLimit=false
    simulationSpace=BoneRelative
    alphaBoolBlend {blendOption=Linear}
    boundBone {boneName=dyn_Tassel_3}
    componentAppliedLinearAccClamp=<3000 3000 3000>
    componentLinearAccScale=<0.2 0.3 0.3>
    componentLinearVelScale=<0.2 0.3 0.3>
    gravityOverride {z=-980}
    physicsBodyDefinitions [{
      boundBone {boneName=dyn_Tassel_3}
      boxExtents=<10 20 20>
      constraintSetup {angularLimitsMax=<20 100 60>, angularLimitsMin=<-20 -100 -60>, angularTarget=<1 0 0>}
      localJointOffset=<2.5 0 0>
     }]
    relativeSpaceBone {boneName=b_AR_Root}
    simSpaceSettings {maxAngularAcceleration=25, maxAngularVelocity=5, simSpaceAngularAlpha=1}
   }
   n12 = Anim Dynamics - Bone: dyn_Tassel_4!(ComponentPose=@n11, Alpha=0.5, GravityOverride:Vector="0,0,-980") -> (Pose) {
    angularDampingOverride=1
    angularSpringConstant=500
    bOverrideAngularDamping=true
    bUseGravityOverride=true
    bUsePlanarLimit=false
    simulationSpace=RootRelative
    alphaBoolBlend {blendOption=Linear}
    boundBone {boneName=dyn_Tassel_4}
    componentAppliedLinearAccClamp=<3000 3000 3000>
    componentLinearAccScale=<0.2 0.3 0.3>
    componentLinearVelScale=<0.2 0.3 0.3>
    gravityOverride {z=-980}
    physicsBodyDefinitions [{
      boundBone {boneName=dyn_Tassel_4}
      boxExtents=<8 15 15>
      constraintSetup {angularLimitsMax=<60 60 60>, angularLimitsMin=<-60 -60 -60>, angularTarget=<1 0 0>}
      localJointOffset=<2.5 0 0>
     }]
    relativeSpaceBone {boneName=deform_Tassel_MSTR}
    simSpaceSettings {maxAngularAcceleration=25, maxAngularVelocity=5, simSpaceAngularAlpha=1}
   }
   n13 = Component To Local(ComponentPose=@n12)
   n14 = Anim Dynamics - Bone: dyn_Tassel_1!(ComponentPose=Local To Component(LocalPose=Use cached pose 'input pose'()), Alpha=1, GravityOverride="0,0,0", ExternalForce:Vector="0,0,0") -> (Pose) {
    angularDampingOverride=0.9994
    bOverrideAngularDamping=true
    bUseGravityOverride=true
    alphaBoolBlend {blendOption=Linear}
    boundBone {boneName=dyn_Tassel_1}
    componentAppliedLinearAccClamp=<2000 2000 2000>
    componentLinearAccScale=<0.4 0.4 0.5>
    componentLinearVelScale=<0.4 0.4 0.5>
    gravityOverride {z=-980}
    physicsBodyDefinitions [{
      boundBone {boneName=dyn_Tassel_1}
      boxExtents=<8 8 8>
      constraintSetup {angularLimitsMax=<20 50 80>, angularLimitsMin=<-20 -50 -90>, angularTarget=<0 0 0>}
      localJointOffset=<1 0 0>
     }]
    simSpaceSettings {maxAngularAcceleration=36, maxAngularVelocity=6, simSpaceAngularAlpha=1}
   }
   n15 = Anim Dynamics - Bone: dyn_Tassel_2!(ComponentPose=@n14, Alpha=1, GravityOverride="0,0,0", ExternalForce:Vector="0,0,0") -> (Pose) {
    angularDampingOverride=0.9994
    bOverrideAngularDamping=true
    bUseGravityOverride=true
    alphaBoolBlend {blendOption=Linear}
    boundBone {boneName=dyn_Tassel_2}
    componentAppliedLinearAccClamp=<2000 2000 2000>
    componentLinearAccScale=<0.4 0.4 0.5>
    componentLinearVelScale=<0.4 0.4 0.5>
    gravityOverride {z=-980}
    physicsBodyDefinitions [{
      boundBone {boneName=dyn_Tassel_2}
      boxExtents=<8 8 8>
      constraintSetup {angularLimitsMax=<40 40 40>, angularLimitsMin=<-40 -40 -40>, angularTarget=<0 0 0>}
      localJointOffset=<1 0 0>
     }]
    planarLimits [{drivingBone {boneName=Melee_TwoHanded_Handle}, planeTransform {location=<0 0 -6>, rotation=<rot 0 0 -180>}}]
    simSpaceSettings {maxAngularAcceleration=36, maxAngularVelocity=6, simSpaceAngularAlpha=1}
   }
   n16 = Anim Dynamics - Bone: dyn_Tassel_3!(ComponentPose=@n15, Alpha=1, GravityOverride="0,0,0", ExternalForce:Vector="0,0,0") -> (Pose) {
    angularDampingOverride=0.9994
    bOverrideAngularDamping=true
    bUseGravityOverride=true
    bUseSphericalLimits=true
    alphaBoolBlend {blendOption=Linear}
    boundBone {boneName=dyn_Tassel_3}
    componentAppliedLinearAccClamp=<2000 2000 2000>
    componentLinearAccScale=<0.4 0.4 0.5>
    componentLinearVelScale=<0.4 0.4 0.5>
    gravityOverride {z=-980}
    physicsBodyDefinitions [{
      boundBone {boneName=dyn_Tassel_3}
      boxExtents=<8 8 8>
      constraintSetup {angularLimitsMax=<40 40 40>, angularLimitsMin=<-40 -40 -40>, angularTarget=<0 0 0>}
      localJointOffset=<1 0 0>
     }]
    planarLimits [{drivingBone {boneName=Melee_TwoHanded_Handle}, planeTransform {location=<0 0 -7>, rotation=<rot 0 0 180>}}]
    simSpaceSettings {maxAngularAcceleration=36, maxAngularVelocity=6, simSpaceAngularAlpha=1}
   }
   n17 = Anim Dynamics - Bone: dyn_Tassel_4!(ComponentPose=@n16, Alpha=1, GravityOverride="0,0,0", ExternalForce:Vector="0,0,0") -> (Pose) {
    angularDampingOverride=0.9994
    bOverrideAngularDamping=true
    bUseGravityOverride=true
    bUseSphericalLimits=true
    alphaBoolBlend {blendOption=Linear}
    boundBone {boneName=dyn_Tassel_4}
    componentAppliedLinearAccClamp=<2000 2000 2000>
    componentLinearAccScale=<0.4 0.4 0.5>
    componentLinearVelScale=<0.4 0.4 0.5>
    gravityOverride {z=-980}
    physicsBodyDefinitions [{
      boundBone {boneName=dyn_Tassel_4}
      boxExtents=<8 8 8>
      constraintSetup {angularLimitsMax=<40 40 40>, angularLimitsMin=<-40 -40 -40>, angularTarget=<0 0 0>}
      localJointOffset=<1 0 0>
     }]
    planarLimits [{drivingBone {boneName=Melee_TwoHanded_Handle}, planeTransform {location=<0 0 -7>, rotation=<rot 0 0 180>}}]
    simSpaceSettings {maxAngularAcceleration=36, maxAngularVelocity=6, simSpaceAngularAlpha=1}
   }
   Save cached pose 'Dynamics_Cache'(Pose=Blend Poses by bool(BlendPose_0=Blend Poses by bool(BlendPose_0=@n7, BlendPose_1=@n13) {blendType=Linear}, BlendPose_1=Component To Local(ComponentPose=@n17)) {blendType=Linear}) {cacheName=Dynamics_Cache}
  ]}
 EventGraph {nodes [
   Event BlueprintUpdateAnimation()
   > n1 = Sequence!(execute) -> (then_0, then_1, then_2:Exec)
   Branch(execute=@n1.then_1, Condition=Get bIsInFrontEnd())
   > Set FrontEndRigidBodies(FrontEndRigidBodies:bool=Get bRigidBodyEnabled()) -> (Output_Get:bool)
   > Set FrontEndAnimDynamics(FrontEndAnimDynamics:bool=NOT Boolean(A=Get bRigidBodyEnabled())) -> (Output_Get:bool)
   n2 = Branch(execute=@n1.then_2, Condition=OR Boolean(A=Get bIsFalling(), B=Get bIsJumping()))
   > Set VelScale(VelScale:Vector="0.4,0.4,0.4") -> (Output_Get:Vector)
   > Branch(Condition=Get bIsFalling())
   > Set VelScale(VelScale="0.35,0.35,0.35")
   Set VelScale(execute=@n2.else, VelScale="0.45,0.45,0.45")
   > Branch(Condition=Get bIsInFrontEnd())
   > Set VelScale(VelScale="0.1,0.25,0.25")
   n3 = Gravity Override State Function DEPRECATED(
    execute=@n1.then_0
    CharacterPartAnimInstance=Self-Reference()
    GameTimeInSeconds=GetGameTimeInSeconds()
    GravityOverrideParams=Get Gravity Override Params()
   )
   > Set Gravity Override(Gravity Override=@n3)
   TryGetPawnOwner()
  ]}
]
variables [
 FrontEndAnimDynamics {type=boolean}
 FrontEndRigidBodies {type=boolean}
 Gravity Override {type=Vector}
 Gravity Override Params {type=GravityOverrideParamsStruct}
 VelScale {type=Vector}
]
